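# Translucent, two-sided unlit material for energy FX on the Snapfreeze glider, built for Niagara sprites, ribbons and mesh particles.
# Three panning noise layers form a contrast mask that blends two colors under a caustic line highlight; a static switch instead maps them to offset RGB channels for color fringing.
# Opacity combines that mask with V fades, a procedural triangle, optional texture mask and view-angle fade, with depth fade skipped on forward and mobile.
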Material /BRCosmetics/Gadgets/Assets/VinderTech_GliderChute/Glider_SnapFreeze/FX/M_Glider_Snapfreeze_Energy
graph {
 properties {
  bAutomaticallySetUsageInEditor=false
  bUsedWithNiagaraMeshParticles=true
  bUsedWithNiagaraRibbons=true
  bUsedWithNiagaraSprites=true
  blendMode=BLEND_Translucent
  shadingModel=MSM_Unlit
  twoSided=true
  usedShadingModels=Unlit
 }
 nodes [
  n1 = ParticleColor()
  n2 = TextureCoordinate()
  n3 = Multiply(A=Constant2Vector() {r=-0.01}, B=ScalarParameter() {defaultValue=1, parameterName=RGBOffsetMultiply})
  n4 = Multiply(
   A=StaticSwitchParameter(True=Add(A=@n2, B=@n3), False=@n2) {parameterName=Use_RGBoffset}
   B=AppendVector(
    A=ScalarParameter() {defaultValue=1, group=EnergyRing01, parameterName=energyRing01_U}
    B=ScalarParameter() {defaultValue=0.2, group=EnergyRing01, parameterName=energyRing01_V}
   )
  )
  n5 = AppendVector(
   A=ScalarParameter() {defaultValue=-0.2, group=EnergyRing01, parameterName=EnergyRingTexture01_SpeedX}
   B=ScalarParameter() {defaultValue=0.5, group=EnergyRing01, parameterName=EnergyRingTexture01_SpeedY}
  )
  n6 = StaticComponentMaskParameter(None=TextureSampleParameter2D(UVs=Panner(Coordinate=@n4, Speed=@n5) {speedX=0.2, speedY=1.5}) {group=EnergyRing01, parameterName=EnergyRingTexture01, texture=&T_NoisePack_01}.RGB) {defaultG=true, group=EnergyRing01, parameterName=MaskChannel01}
  n7 = TextureCoordinate()
  n8 = Constant2Vector()
  n9 = Multiply(
   A=StaticSwitchParameter(True=Add(A=@n7, B=@n8), False=@n7) {parameterName=Use_RGBoffset}
   B=AppendVector(
    A=ScalarParameter() {defaultValue=3, group=EnergyRing02, parameterName=energyRing02_U}
    B=ScalarParameter() {defaultValue=0.4, group=EnergyRing02, parameterName=energyRing02_V}
   )
  )
  n10 = AppendVector(
   A=ScalarParameter() {defaultValue=0.2, group=EnergyRing02, parameterName=EnergyRingTexture02_SpeedX}
   B=ScalarParameter() {defaultValue=1.5, group=EnergyRing02, parameterName=EnergyRingTexture02_SpeedY}
  )
  n11 = StaticComponentMaskParameter(None=TextureSampleParameter2D(UVs=Panner(Coordinate=@n9, Speed=@n10) {speedX=0.2, speedY=1.5}) {group=EnergyRing02, parameterName=EnergyRingTexture02, texture=&T_NoisePack_01}.RGB) {defaultG=true, group=EnergyRing02, parameterName=MaskChannel02}
  n12 = AppendVector(A=@n6, B=@n11)
  n13 = TextureCoordinate()
  n14 = Multiply(A=Constant2Vector() {r=0.01}, B=ScalarParameter() {defaultValue=1, parameterName=RGBOffsetMultiply})
  n15 = Multiply(
   A=StaticSwitchParameter(True=Add(A=@n13, B=@n14), False=@n13) {parameterName=Use_RGBoffset}
   B=AppendVector(
    A=ScalarParameter() {defaultValue=4, group=EnergyRing03, parameterName=energyRing03_U}
    B=ScalarParameter() {defaultValue=0.5, group=EnergyRing03, parameterName=energyRing03_V}
   )
  )
  n16 = AppendVector(
   A=ScalarParameter() {group=EnergyRing03, parameterName=EnergyRingTexture03_SpeedX}
   B=ScalarParameter() {defaultValue=2, group=EnergyRing03, parameterName=EnergyRingTexture03_SpeedY}
  )
  n17 = StaticComponentMaskParameter(None=TextureSampleParameter2D(UVs=Panner(Coordinate=@n15, Speed=@n16) {speedY=2}) {group=EnergyRing03, parameterName=EnergyRingTexture03, texture=&T_NoisePack_01}.RGB) {defaultG=true, group=EnergyRing03, parameterName=MaskChannel03}
  n18 = MaterialFunctionCall() {materialFunction=&LinearGradient}
  n19 = Power(Base=@n18.VGradient) {constExponent=5}
  n20 = ScalarParameter() {defaultValue=1, parameterName=Linear_Light}
  n21 = Desaturation(None=Add(A=AppendVector(A=@n12, B=@n17), B=Multiply(A=@n19, B=@n20)), Fraction=ScalarParameter() {parameterName=Desaturation})
  n22 = ScalarParameter() {defaultValue=1, parameterName=RGBOffset_light}
  n23 = Multiply(A=Power(Base=Multiply(A=@n21, B=@n22), Exp=ScalarParameter() {defaultValue=1, parameterName=RGBOffset_power}), B=@n1.RGB)
  n24 = VectorParameter() {group=EnergyRing, parameterName=Color_A, defaultValue=<rgba 5 0 0.0884 1>}
  n25 = VectorParameter() {group=EnergyRing, parameterName=Color_B, defaultValue=<rgba 5 1.3 0 1>}
  n26 = Add(A=Multiply(A=Add(A=@n6, B=@n11), B=@n17), B=@n19)
  n27 = Clamp(None=MaterialFunctionCall(In=Power(Base=@n26) {constExponent=0.5}, Contrast=ScalarParameter() {defaultValue=0.5, group=EnergyRing, parameterName=HardEdgeness}) {materialFunction=&CheapContrast}.Result)
  n28 = VectorParameter() {group=EnergyRing, parameterName=Color_Line, defaultValue=<rgba 100 10 0 1>}
  n29 = Multiply(
   A=TextureCoordinate()
   B=AppendVector(
    A=ScalarParameter() {defaultValue=3, group=EnergyRing, parameterName=energyRingLine_U}
    B=ScalarParameter() {defaultValue=0.3, group=EnergyRing, parameterName=energyRingLine_V}
   )
  )
  n30 = AppendVector(
   A=ScalarParameter() {group=EnergyRing, parameterName=EnergyRingTexture_Line_SpeedX}
   B=ScalarParameter() {defaultValue=0.5, group=EnergyRing, parameterName=EnergyRingTexture_Line_SpeedY}
  )
  n31 = Multiply(
   A=Power(Base=@n18.VGradient) {constExponent=3}
   B=TextureSampleParameter2D(UVs=Panner(Coordinate=@n29, Speed=@n30) {speedY=0.5}) {parameterName=EnergyRingTexture_Line, texture=&T_ShieldCaustics}.R
  )
  n32 = Multiply(A=Add(A=LinearInterpolate(A=@n24.RGB, B=@n25.RGB, Alpha=@n27), B=Multiply(A=@n28.RGB, B=@n31)), B=@n1.RGB)
  MaterialFunctionCall(Emissive=Multiply(A=StaticSwitchParameter(True=@n23, False=@n32) {parameterName=Use_RGBoffset}, B=ScalarParameter() {defaultValue=1, parameterName=All light})) {materialFunction=&MF_FX_EmissiveAdjustments}
  n33 = TextureCoordinate() {uTiling=0.5, vTiling=0.5}
  n34 = Sine(None=ComponentMask(None=@n33) {g=true})
  n35 = ScalarParameter() {defaultValue=2, parameterName=BaseFade2}
  n36 = Power(Base=Multiply(A=ComponentMask(None=@n33) {g=true}) {constB=5}) {constExponent=1}
  n37 = Multiply(A=Multiply(A=@n27, B=@n34), B=Multiply(A=Power(Base=@n34, Exp=@n35), B=Saturate(None=@n36)))
  n38 = MaterialFunctionCall(
   UVs=TextureCoordinate()
   Texture Scale=MaterialFunctionCall(
    X=ScalarParameter() {defaultValue=1.5, group=Shape, parameterName=TriangleWidth}
    Y=ScalarParameter() {defaultValue=1.5, group=Shape, parameterName=TriangleHeight}
   ) {materialFunction=&MakeFloat2}.Result
  ) {materialFunction=&ScaleUVsByCenter}
  n39 = ComponentMask(None=@n38.UVs) {g=true}
  n40 = Divide(A=ConstantBiasScale(None=ComponentMask(None=@n38.UVs) {r=true}) {bias=-0.5, scale=2}, B=OneMinus(None=OneMinus(None=@n39)))
  n41 = DotProduct(A=@n40, B=@n40)
  n42 = SquareRoot(None=OneMinus(None=Saturate(None=@n41)))
  n43 = OneMinus(None=@n39)
  n44 = ScalarParameter() {defaultValue=0.05, group=Shape, parameterName=Triangle Base Fade}
  n45 = Multiply(
   A=Power(Base=Saturate(None=@n42), Exp=ScalarParameter() {defaultValue=4, group=Shape, parameterName=Triangle Edge Falloff})
   B=Power(
    Base=MaterialFunctionCall!(Gradient=@n43, Center Point=@n44) {materialFunction=/Game/Effects/Fort_Effects/MaterialFunctions/MF_GradientReflected}.Result
    Exp=ScalarParameter() {defaultValue=2, group=Shape, parameterName=Triangle Length Falloff}
   )
  )
  n46 = Multiply(
   A=@n1.A
   B=LinearInterpolate(A=@n37, B=Multiply(A=@n37, B=@n45), Alpha=ScalarParameter() {defaultValue=1, group=Shape, parameterName=TriangleSHape})
  )
  n47 = ScalarParameter() {defaultValue=1, group=MaskTex, parameterName=MaskTexture_U}
  n48 = ScalarParameter() {defaultValue=1, group=MaskTex, parameterName=MaskTexture_V}
  n49 = TextureSampleParameter2D(UVs=Multiply(A=TextureCoordinate(), B=AppendVector(A=@n47, B=@n48))) {
   group=MaskTex
   parameterName=MaskTexture
   samplerType=SAMPLERTYPE_Grayscale
   texture=/Game/Effects/Fort_Effects/Textures/Smoke/T_Smoke_Trail_Soft_NoClamp
  }
  n50 = Power(Base=@n49.R, Exp=ScalarParameter() {defaultValue=1, group=MaskTex, parameterName=MaskTexture_power})
  n51 = Multiply(
   A=LinearInterpolate(A=Multiply(A=@n46, B=@n50), B=@n46, Alpha=ScalarParameter() {defaultValue=1, group=MaskTex, parameterName=Use_MaskTexture})
   B=MaterialFunctionCall("FadeSensitivity (optional)"=ScalarParameter() {defaultValue=1.5, parameterName=MeshFade}) {materialFunction=&FadeBasedOnViewAngle}.Result
  )
  n52 = DepthFade(Opacity=@n51, FadeDistance=ScalarParameter() {defaultValue=5, parameterName=DepthFade})
  Multiply(
   A=ShadingPathSwitch(Default=@n52, Deferred=@n52, Forward=@n51, Mobile=@n51)
   B=ScalarParameter() {defaultValue=1, group=MaskTex, parameterName=MasterOpacity}
  )
 ]
}
